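# Post-process AnimBP for the SleekRivet medium male cosmetic head, adding secondary motion to ears, earring and whisker bases.
# AnimDynamics drives both ears and the left earring, then weighted constraints blend the earring between head and ear, and whisker bases across facial bones, before a head yaw offset.
# Six whisker AnimDynamics nodes form a separate chain unconnected to the output.
# The EventGraph computes emote, celebration and head-rotation values that the AnimGraph never reads.
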AnimBlueprint /BRCosmetics/Characters/Player/Male/Medium/Heads/M_MED_SleekRivet_Head/Meshes/M_MED_SleekRivet_Head_AnimBP_postprocess
target_skeleton=&/M_MED_SleekRivet_Head_Skeleton
graphs [
 AnimGraph {nodes [
   n1 = Anim Dynamics - Bone: dyn_whiske_mid_l_01(Alpha=1, ExternalForce="0,0,0") {
    angularDampingOverride=1
    angularSpringConstant=10000
    bOverrideAngularDamping=true
    bUseGravityOverride=true
    gravityScale=0.35
    simulationSpace=RootRelative
    alphaBoolBlend {blendOption=Linear}
    boundBone {boneName=dyn_whiske_mid_l_01}
    componentAppliedLinearAccClamp=<6000 6000 6000>
    componentLinearAccScale=<0.3 0.3 0.3>
    componentLinearVelScale=<0.3 0.3 0.3>
    gravityOverride {z=-980}
    physicsBodyDefinitions [{
      boundBone {boneName=dyn_whiske_mid_l_01}
      boxExtents=<15 16 15>
      constraintSetup {coneAngle=1, angularLimitsMax=<15 22 15>, angularLimitsMin=<-15 -22 -15>, angularTarget=<1 0 0>}
      localJointOffset=<1 0 0>
     }]
   }
   n2 = Anim Dynamics - Bone: dyn_whiske_up_l_01(ComponentPose=@n1, Alpha=1, ExternalForce="0,0,0") {
    angularDampingOverride=1
    angularSpringConstant=10000
    bOverrideAngularDamping=true
    bUseGravityOverride=true
    gravityScale=0.35
    simulationSpace=RootRelative
    alphaBoolBlend {blendOption=Linear}
    boundBone {boneName=dyn_whiske_up_l_01}
    componentAppliedLinearAccClamp=<6000 6000 6000>
    componentLinearAccScale=<0.3 0.3 0.3>
    componentLinearVelScale=<0.3 0.3 0.3>
    gravityOverride {z=-980}
    physicsBodyDefinitions [{
      boundBone {boneName=dyn_whiske_up_l_01}
      boxExtents=<15 16 15>
      constraintSetup {coneAngle=1, angularLimitsMax=<15 22 15>, angularLimitsMin=<-15 -22 -15>, angularTarget=<1 0 0>}
      localJointOffset=<1 0 0>
     }]
   }
   n3 = Anim Dynamics - Bone: dyn_whiske_dwn_l_01(ComponentPose=@n2, Alpha=1, ExternalForce="0,0,0") {
    angularDampingOverride=1
    angularSpringConstant=10000
    bOverrideAngularDamping=true
    bUseGravityOverride=true
    gravityScale=0.35
    simulationSpace=RootRelative
    alphaBoolBlend {blendOption=Linear}
    boundBone {boneName=dyn_whiske_dwn_l_01}
    componentAppliedLinearAccClamp=<6000 6000 6000>
    componentLinearAccScale=<0.3 0.3 0.3>
    componentLinearVelScale=<0.3 0.3 0.3>
    gravityOverride {z=-980}
    physicsBodyDefinitions [{
      boundBone {boneName=dyn_whiske_dwn_l_01}
      boxExtents=<15 16 15>
      constraintSetup {coneAngle=1, angularLimitsMax=<15 22 15>, angularLimitsMin=<-15 -22 -15>, angularTarget=<1 0 0>}
      localJointOffset=<1 0 0>
     }]
   }
   n4 = Anim Dynamics - Bone: dyn_whiske_mid_r_01(ComponentPose=@n3, Alpha=1, ExternalForce="0,0,0") {
    angularDampingOverride=1
    angularSpringConstant=10000
    bOverrideAngularDamping=true
    bUseGravityOverride=true
    gravityScale=0.35
    simulationSpace=RootRelative
    alphaBoolBlend {blendOption=Linear}
    boundBone {boneName=dyn_whiske_mid_r_01}
    componentAppliedLinearAccClamp=<6000 6000 6000>
    componentLinearAccScale=<0.3 0.3 0.3>
    componentLinearVelScale=<0.3 0.3 0.3>
    gravityOverride {z=-980}
    physicsBodyDefinitions [{
      boundBone {boneName=dyn_whiske_mid_r_01}
      boxExtents=<15 16 15>
      constraintSetup {coneAngle=1, angularLimitsMax=<15 22 15>, angularLimitsMin=<-15 -22 -15>, angularTarget=<1 0 0>}
      localJointOffset=<-1 0 0>
     }]
   }
   n5 = Anim Dynamics - Bone: dyn_whiske_up_r_01(ComponentPose=@n4, Alpha=1, ExternalForce="0,0,0") {
    angularDampingOverride=1
    angularSpringConstant=10000
    bOverrideAngularDamping=true
    bUseGravityOverride=true
    gravityScale=0.35
    simulationSpace=RootRelative
    alphaBoolBlend {blendOption=Linear}
    boundBone {boneName=dyn_whiske_up_r_01}
    componentAppliedLinearAccClamp=<6000 6000 6000>
    componentLinearAccScale=<0.3 0.3 0.3>
    componentLinearVelScale=<0.3 0.3 0.3>
    gravityOverride {z=-980}
    physicsBodyDefinitions [{
      boundBone {boneName=dyn_whiske_up_r_01}
      boxExtents=<15 16 15>
      constraintSetup {coneAngle=1, angularLimitsMax=<15 22 15>, angularLimitsMin=<-15 -22 -15>, angularTarget=<1 0 0>}
      localJointOffset=<-1 0 0>
     }]
   }
   Anim Dynamics - Bone: dyn_whiske_dwn_r_01(ComponentPose=@n5, Alpha=1, ExternalForce="0,0,0") {
    angularDampingOverride=1
    angularSpringConstant=10000
    bOverrideAngularDamping=true
    bUseGravityOverride=true
    gravityScale=0.35
    simulationSpace=RootRelative
    alphaBoolBlend {blendOption=Linear}
    boundBone {boneName=dyn_whiske_dwn_r_01}
    componentAppliedLinearAccClamp=<6000 6000 6000>
    componentLinearAccScale=<0.3 0.3 0.3>
    componentLinearVelScale=<0.3 0.3 0.3>
    gravityOverride {z=-980}
    physicsBodyDefinitions [{
      boundBone {boneName=dyn_whiske_dwn_r_01}
      boxExtents=<15 16 15>
      constraintSetup {coneAngle=1, angularLimitsMax=<15 22 15>, angularLimitsMin=<-15 -22 -15>, angularTarget=<1 0 0>}
      localJointOffset=<-1 0 0>
     }]
   }
   n6 = "Transform (Modify) Bone - Bone: head"(ComponentPose=Local To Component(LocalPose=Input Pose()), Alpha=1, Translation="0,0,0", Rotation="0,0,0", Scale="1,1,1") {
    rotationMode=BMM_Additive
    rotationSpace=BCS_BoneSpace
    translationMode=BMM_Additive
    alphaBoolBlend {blendOption=Linear}
    boneToModify {boneName=head}
   }
   n7 = Anim Dynamics - Bone: dyn_ear_r_01(ComponentPose=@n6, Alpha=1, ExternalForce="0,0,0") {
    angularDampingOverride=1
    angularSpringConstant=25
    bAngularSpring=true
    bOverrideAngularDamping=true
    bUseGravityOverride=true
    gravityScale=0.35
    simulationSpace=RootRelative
    alphaBoolBlend {blendOption=Linear}
    boundBone {boneName=dyn_ear_r_01}
    componentAppliedLinearAccClamp=<6000 6000 6000>
    componentLinearAccScale=<0.3 0.3 0.3>
    componentLinearVelScale=<0.3 0.3 0.3>
    gravityOverride {z=-980}
    physicsBodyDefinitions [{
      boundBone {boneName=dyn_ear_r_01}
      boxExtents=<15 20 20>
      constraintSetup {angularLimitsMax=<15 15 15>, angularLimitsMin=<-15 -15 -15>, angularTarget=<1 0 0>}
      localJointOffset=<2.5 0 0>
     }]
    relativeSpaceBone {boneName=head}
   }
   n8 = Anim Dynamics - Bone: dyn_ear_l_01(ComponentPose=@n7, Alpha=1, ExternalForce="0,0,0") {
    angularDampingOverride=1
    angularSpringConstant=25
    bAngularSpring=true
    bOverrideAngularDamping=true
    bUseGravityOverride=true
    gravityScale=0.35
    simulationSpace=RootRelative
    alphaBoolBlend {blendOption=Linear}
    boundBone {boneName=dyn_ear_l_01}
    componentAppliedLinearAccClamp=<6000 6000 6000>
    componentLinearAccScale=<0.3 0.3 0.3>
    componentLinearVelScale=<0.3 0.3 0.3>
    gravityOverride {z=-980}
    physicsBodyDefinitions [{
      boundBone {boneName=dyn_ear_l_01}
      boxExtents=<15 20 20>
      constraintSetup {angularLimitsMax=<15 15 15>, angularLimitsMin=<-15 -15 -15>, angularTarget=<1 0 0>}
      localJointOffset=<2.5 0 0>
     }]
    relativeSpaceBone {boneName=head}
   }
   n9 = Anim Dynamics - Bone: dyn_earing_l_MSTR1(ComponentPose=@n8, Alpha=1, ExternalForce="0,0,0") {
    angularDampingOverride=1
    angularSpringConstant=500
    bOverrideAngularDamping=true
    bUseGravityOverride=true
    gravityScale=0.35
    simulationSpace=RootRelative
    alphaBoolBlend {blendOption=Linear}
    boundBone {boneName=dyn_earing_l_MSTR1}
    componentAppliedLinearAccClamp=<6000 6000 6000>
    componentLinearAccScale=<0.3 0.3 0.3>
    componentLinearVelScale=<0.3 0.3 0.3>
    gravityOverride {z=-980}
    physicsBodyDefinitions [{
      boundBone {boneName=dyn_earing_l_MSTR1}
      boxExtents=<8 8 8>
      constraintSetup {coneAngle=1, angularLimitsMax=<0 11 0>, angularLimitsMin=<0 -35 0>, angularTarget=<1 0 0>}
      localJointOffset=<1 0 0>
     }]
   }
   n10 = "Constraint - dyn_earing_l_MSTR1 (PP)"!(ComponentPose=@n9, Alpha=1, ConstraintWeights_0=0.7, ConstraintWeights_1:float=0.3) -> (Pose) {
    alphaBoolBlend {blendOption=Linear}
    boneToModify {boneName=dyn_earing_l_MSTR1}
    constraintSetup [{targetBone {boneName=head}}, {targetBone {boneName=dyn_ear_l_01}}]
    constraintWeights [1,1]
   }
   n11 = "Constraint - Base_whiske_up_l (PPPPPPPPPPP)"!(
    ComponentPose=@n10
    Alpha=1
    ConstraintWeights_0=0.648
    ConstraintWeights_1:float=0.0006
    ConstraintWeights_2:float=0
    ConstraintWeights_3:float=0
    ConstraintWeights_4:float=0.0688
    ConstraintWeights_5:float=0.1242
    ConstraintWeights_6:float=0.0003
    ConstraintWeights_7:float=0.0272
    ConstraintWeights_8:float=0.1308
    ConstraintWeights_9:float=0
    ConstraintWeights_10:float=0.0001
   ) -> (Pose) {
    alphaBoolBlend {blendOption=Linear}
    boneToModify {boneName=Base_whiske_up_l}
    constraintSetup [
     {targetBone {boneName=head}}
     {targetBone {boneName=C_jaw}}
     {targetBone {boneName=L_lip_lower_outer}}
     {targetBone {boneName=R_lip_lower_outer}}
     {targetBone {boneName=C_lip_upper_mid}}
     {targetBone {boneName=L_cheek_inner}}
     {targetBone {boneName=L_eye_lid_lower_mid}}
     {targetBone {boneName=L_lip_corner}}
     {targetBone {boneName=L_lip_upper_outer}}
     {targetBone {boneName=R_cheek_inner}}
     {targetBone {boneName=R_lip_upper_outer}}
    ]
    constraintWeights [1,1,1,1,1,1,1,1,1,1,1]
   }
   n12 = "Constraint - Base_whiske_mid_l (PPPPPPPPPPP)"!(
    ComponentPose=@n11
    Alpha=1
    ConstraintWeights_0=0.648
    ConstraintWeights_1:float=0.0006
    ConstraintWeights_2:float=0
    ConstraintWeights_3:float=0
    ConstraintWeights_4:float=0.0688
    ConstraintWeights_5:float=0.1242
    ConstraintWeights_6:float=0.0003
    ConstraintWeights_7:float=0.0272
    ConstraintWeights_8:float=0.1308
    ConstraintWeights_9:float=0
    ConstraintWeights_10:float=0.0001
   ) -> (Pose) {
    alphaBoolBlend {blendOption=Linear}
    boneToModify {boneName=Base_whiske_mid_l}
    constraintSetup [
     {targetBone {boneName=head}}
     {targetBone {boneName=C_jaw}}
     {targetBone {boneName=L_lip_lower_outer}}
     {targetBone {boneName=R_lip_lower_outer}}
     {targetBone {boneName=C_lip_upper_mid}}
     {targetBone {boneName=L_cheek_inner}}
     {targetBone {boneName=L_eye_lid_lower_mid}}
     {targetBone {boneName=L_lip_corner}}
     {targetBone {boneName=L_lip_upper_outer}}
     {targetBone {boneName=R_cheek_inner}}
     {targetBone {boneName=R_lip_upper_outer}}
    ]
    constraintWeights [1,1,1,1,1,1,1,1,1,1,1]
   }
   n13 = "Constraint - Base_whiske_dwn_l (PPPPPPPPPPP)"!(
    ComponentPose=@n12
    Alpha=1
    ConstraintWeights_0=0.648
    ConstraintWeights_1:float=0.0006
    ConstraintWeights_2:float=0
    ConstraintWeights_3:float=0
    ConstraintWeights_4:float=0.0688
    ConstraintWeights_5:float=0.1242
    ConstraintWeights_6:float=0.0003
    ConstraintWeights_7:float=0.0272
    ConstraintWeights_8:float=0.1308
    ConstraintWeights_9:float=0
    ConstraintWeights_10:float=0.0001
   ) -> (Pose) {
    alphaBoolBlend {blendOption=Linear}
    boneToModify {boneName=Base_whiske_dwn_l}
    constraintSetup [
     {targetBone {boneName=head}}
     {targetBone {boneName=C_jaw}}
     {targetBone {boneName=L_lip_lower_outer}}
     {targetBone {boneName=R_lip_lower_outer}}
     {targetBone {boneName=C_lip_upper_mid}}
     {targetBone {boneName=L_cheek_inner}}
     {targetBone {boneName=L_eye_lid_lower_mid}}
     {targetBone {boneName=L_lip_corner}}
     {targetBone {boneName=L_lip_upper_outer}}
     {targetBone {boneName=R_cheek_inner}}
     {targetBone {boneName=R_lip_upper_outer}}
    ]
    constraintWeights [1,1,1,1,1,1,1,1,1,1,1]
   }
   n14 = "Constraint - Base_whiske_dwn_l_2 (PPPPPPPPPPP)"!(
    ComponentPose=@n13
    Alpha=1
    ConstraintWeights_0=0.648
    ConstraintWeights_1:float=0.0006
    ConstraintWeights_2:float=0
    ConstraintWeights_3:float=0
    ConstraintWeights_4:float=0.0688
    ConstraintWeights_5:float=0.1242
    ConstraintWeights_6:float=0.0003
    ConstraintWeights_7:float=0.0272
    ConstraintWeights_8:float=0.1308
    ConstraintWeights_9:float=0
    ConstraintWeights_10:float=0.0001
   ) -> (Pose) {
    alphaBoolBlend {blendOption=Linear}
    boneToModify {boneName=Base_whiske_dwn_l_2}
    constraintSetup [
     {targetBone {boneName=head}}
     {targetBone {boneName=C_jaw}}
     {targetBone {boneName=L_lip_lower_outer}}
     {targetBone {boneName=R_lip_lower_outer}}
     {targetBone {boneName=C_lip_upper_mid}}
     {targetBone {boneName=L_cheek_inner}}
     {targetBone {boneName=L_eye_lid_lower_mid}}
     {targetBone {boneName=L_lip_corner}}
     {targetBone {boneName=L_lip_upper_outer}}
     {targetBone {boneName=R_cheek_inner}}
     {targetBone {boneName=R_lip_upper_outer}}
    ]
    constraintWeights [1,1,1,1,1,1,1,1,1,1,1]
   }
   n15 = "Constraint - Base_whiske_dwn_l_3 (PPPPPPPPPPP)"!(
    ComponentPose=@n14
    Alpha=1
    ConstraintWeights_0=0.648
    ConstraintWeights_1:float=0.0006
    ConstraintWeights_2:float=0
    ConstraintWeights_3:float=0
    ConstraintWeights_4:float=0.0688
    ConstraintWeights_5:float=0.1242
    ConstraintWeights_6:float=0.0003
    ConstraintWeights_7:float=0.0272
    ConstraintWeights_8:float=0.1308
    ConstraintWeights_9:float=0
    ConstraintWeights_10:float=0.0001
   ) -> (Pose) {
    alphaBoolBlend {blendOption=Linear}
    boneToModify {boneName=Base_whiske_dwn_l_3}
    constraintSetup [
     {targetBone {boneName=head}}
     {targetBone {boneName=C_jaw}}
     {targetBone {boneName=L_lip_lower_outer}}
     {targetBone {boneName=R_lip_lower_outer}}
     {targetBone {boneName=C_lip_upper_mid}}
     {targetBone {boneName=L_cheek_inner}}
     {targetBone {boneName=L_eye_lid_lower_mid}}
     {targetBone {boneName=L_lip_corner}}
     {targetBone {boneName=L_lip_upper_outer}}
     {targetBone {boneName=R_cheek_inner}}
     {targetBone {boneName=R_lip_upper_outer}}
    ]
    constraintWeights [1,1,1,1,1,1,1,1,1,1,1]
   }
   n16 = "Constraint - Base_whiske_up_r (PPPPPPPPPPP)"!(
    ComponentPose=@n15
    Alpha=1
    ConstraintWeights_0=0.6299
    ConstraintWeights_1:float=0.0006
    ConstraintWeights_2:float=0
    ConstraintWeights_3:float=0
    ConstraintWeights_4:float=0.0694
    ConstraintWeights_5:float=0
    ConstraintWeights_6:float=0.0001
    ConstraintWeights_7:float=0.132
    ConstraintWeights_8:float=0.0003
    ConstraintWeights_9:float=0.0292
    ConstraintWeights_10:float=0.1385
   ) -> (Pose) {
    alphaBoolBlend {blendOption=Linear}
    boneToModify {boneName=Base_whiske_up_r}
    constraintSetup [
     {targetBone {boneName=head}}
     {targetBone {boneName=C_jaw}}
     {targetBone {boneName=L_lip_lower_outer}}
     {targetBone {boneName=R_lip_lower_outer}}
     {targetBone {boneName=C_lip_upper_mid}}
     {targetBone {boneName=L_cheek_inner}}
     {targetBone {boneName=L_lip_upper_outer}}
     {targetBone {boneName=R_cheek_inner}}
     {targetBone {boneName=R_eye_lid_lower_mid}}
     {targetBone {boneName=R_lip_corner}}
     {targetBone {boneName=R_lip_upper_outer}}
    ]
    constraintWeights [1,1,1,1,1,1,1,1,1,1,1]
   }
   n17 = "Constraint - Base_whiske_mid_r (PPPPPPPPPPP)"!(
    ComponentPose=@n16
    Alpha=1
    ConstraintWeights_0=0.6299
    ConstraintWeights_1:float=0.0006
    ConstraintWeights_2:float=0
    ConstraintWeights_3:float=0
    ConstraintWeights_4:float=0.0694
    ConstraintWeights_5:float=0
    ConstraintWeights_6:float=0.0001
    ConstraintWeights_7:float=0.132
    ConstraintWeights_8:float=0.0003
    ConstraintWeights_9:float=0.0292
    ConstraintWeights_10:float=0.1385
   ) -> (Pose) {
    alphaBoolBlend {blendOption=Linear}
    boneToModify {boneName=Base_whiske_mid_r}
    constraintSetup [
     {targetBone {boneName=head}}
     {targetBone {boneName=C_jaw}}
     {targetBone {boneName=L_lip_lower_outer}}
     {targetBone {boneName=R_lip_lower_outer}}
     {targetBone {boneName=C_lip_upper_mid}}
     {targetBone {boneName=L_cheek_inner}}
     {targetBone {boneName=L_lip_upper_outer}}
     {targetBone {boneName=R_cheek_inner}}
     {targetBone {boneName=R_eye_lid_lower_mid}}
     {targetBone {boneName=R_lip_corner}}
     {targetBone {boneName=R_lip_upper_outer}}
    ]
    constraintWeights [1,1,1,1,1,1,1,1,1,1,1]
   }
   n18 = "Constraint - Base_whiske_dwn_r (PPPPPPPPPPP)"!(
    ComponentPose=@n17
    Alpha=1
    ConstraintWeights_0=0.6299
    ConstraintWeights_1:float=0.0006
    ConstraintWeights_2:float=0
    ConstraintWeights_3:float=0
    ConstraintWeights_4:float=0.0694
    ConstraintWeights_5:float=0
    ConstraintWeights_6:float=0.0001
    ConstraintWeights_7:float=0.132
    ConstraintWeights_8:float=0.0003
    ConstraintWeights_9:float=0.0292
    ConstraintWeights_10:float=0.1385
   ) -> (Pose) {
    alphaBoolBlend {blendOption=Linear}
    boneToModify {boneName=Base_whiske_dwn_r}
    constraintSetup [
     {targetBone {boneName=head}}
     {targetBone {boneName=C_jaw}}
     {targetBone {boneName=L_lip_lower_outer}}
     {targetBone {boneName=R_lip_lower_outer}}
     {targetBone {boneName=C_lip_upper_mid}}
     {targetBone {boneName=L_cheek_inner}}
     {targetBone {boneName=L_lip_upper_outer}}
     {targetBone {boneName=R_cheek_inner}}
     {targetBone {boneName=R_eye_lid_lower_mid}}
     {targetBone {boneName=R_lip_corner}}
     {targetBone {boneName=R_lip_upper_outer}}
    ]
    constraintWeights [1,1,1,1,1,1,1,1,1,1,1]
   }
   n19 = "Constraint - Base_whiske_dwn_r_2 (PPPPPPPPPPP)"!(
    ComponentPose=@n18
    Alpha=1
    ConstraintWeights_0=0.6299
    ConstraintWeights_1:float=0.0006
    ConstraintWeights_2:float=0
    ConstraintWeights_3:float=0
    ConstraintWeights_4:float=0.0694
    ConstraintWeights_5:float=0
    ConstraintWeights_6:float=0.0001
    ConstraintWeights_7:float=0.132
    ConstraintWeights_8:float=0.0003
    ConstraintWeights_9:float=0.0292
    ConstraintWeights_10:float=0.1385
   ) -> (Pose) {
    alphaBoolBlend {blendOption=Linear}
    boneToModify {boneName=Base_whiske_dwn_r_2}
    constraintSetup [
     {targetBone {boneName=head}}
     {targetBone {boneName=C_jaw}}
     {targetBone {boneName=L_lip_lower_outer}}
     {targetBone {boneName=R_lip_lower_outer}}
     {targetBone {boneName=C_lip_upper_mid}}
     {targetBone {boneName=L_cheek_inner}}
     {targetBone {boneName=L_lip_upper_outer}}
     {targetBone {boneName=R_cheek_inner}}
     {targetBone {boneName=R_eye_lid_lower_mid}}
     {targetBone {boneName=R_lip_corner}}
     {targetBone {boneName=R_lip_upper_outer}}
    ]
    constraintWeights [1,1,1,1,1,1,1,1,1,1,1]
   }
   n20 = "Constraint - Base_whiske_dwn_r_3 (PPPPPPPPPPP)"!(
    ComponentPose=@n19
    Alpha=1
    ConstraintWeights_0=0.6299
    ConstraintWeights_1:float=0.0006
    ConstraintWeights_2:float=0
    ConstraintWeights_3:float=0
    ConstraintWeights_4:float=0.0694
    ConstraintWeights_5:float=0
    ConstraintWeights_6:float=0.0001
    ConstraintWeights_7:float=0.132
    ConstraintWeights_8:float=0.0003
    ConstraintWeights_9:float=0.0292
    ConstraintWeights_10:float=0.1385
   ) -> (Pose) {
    alphaBoolBlend {blendOption=Linear}
    boneToModify {boneName=Base_whiske_dwn_r_3}
    constraintSetup [
     {targetBone {boneName=head}}
     {targetBone {boneName=C_jaw}}
     {targetBone {boneName=L_lip_lower_outer}}
     {targetBone {boneName=R_lip_lower_outer}}
     {targetBone {boneName=C_lip_upper_mid}}
     {targetBone {boneName=L_cheek_inner}}
     {targetBone {boneName=L_lip_upper_outer}}
     {targetBone {boneName=R_cheek_inner}}
     {targetBone {boneName=R_eye_lid_lower_mid}}
     {targetBone {boneName=R_lip_corner}}
     {targetBone {boneName=R_lip_upper_outer}}
    ]
    constraintWeights [1,1,1,1,1,1,1,1,1,1,1]
   }
   n21 = "Transform (Modify) Bone - Bone: head"!(ComponentPose=@n20, Alpha=1, Rotation="0,-22,0") -> (Pose) {
    rotationMode=BMM_Additive
    rotationSpace=BCS_BoneSpace
    translationMode=BMM_Additive
    alphaBoolBlend {blendOption=Linear}
    boneToModify {boneName=head}
    rotation {yaw=-22}
   }
   Output Pose(Result=Component To Local(ComponentPose=@n21))
  ]}
 EventGraph {nodes [
   n1 = Event BlueprintUpdateAnimation()
   > n2 = Sequence()
   n3 = GetOwnerSkeletalMeshComponent()
   n4 = Is Valid(exec=@n2.then_1, InputObject=@n3)
   n5 = GetAnimInstance(self=@n3)
   n6 = FInterpTo(
    Current=Get Celeb up alpha() -> (Celeb up alpha:double)
    Target="To Float (Boolean)"(InBool=Montage_IsPlaying(self=@n5, Montage=/Game/Animation/Game/MainPlayer/Emotes/Emote_Celebration_M))
    DeltaTime=float * float(A=@n1.DeltaTimeX, B=0.3) {is=title K2Node_PromotableOperator}
    InterpSpeed=30
   )
   > Set Celeb up alpha(Celeb up alpha:double=@n6) -> (Output_Get:double)
   Set Celeb up alpha(execute=@n4.Is Not Valid, Celeb up alpha=@n6)
   n7 = GetOwnerSkeletalMeshComponent()
   n8 = BreakRotator(InRot=BreakTransform(InTransform=GetDeltaTransformFromRefPose(self=@n7, BoneName=head, BaseName=neck_02)).Rotation)
   n9 = Get Z percentageEmote() -> (Z percentageEmote:float)
   n10 = Get X percentageEmote() -> (X percentageEmote:float)
   n11 = Get Y percentageEmote() -> (Y percentageEmote:float)
   Set headRotLimit(
    execute=@n2.then_0
    headRotLimit:Rotator=MakeRotator(
     Roll=float * float!(A=@n8, B=-1, C:double=@n10) -> (ReturnValue) {is=title K2Node_CommutativeAssociativeBinaryOperator}
     Pitch=float * float(A=@n8.Pitch, B=-1, C=@n11)
     Yaw=float * float(A=@n8.Yaw, B=-1, C=@n9)
    )
   ) -> (Output_Get:Rotator)
   > n12 = Branch(Condition=Get bIsPlayingEmote())
   n13 = FInterpTo(
    Current=Get Is emote?() -> (Is emote?:double)
    Target=1
    DeltaTime=float * float(A=@n1.DeltaTimeX, B=0.3) {is=title K2Node_PromotableOperator}
    InterpSpeed=10
   )
   > Set Is emote?(Is emote?:double=@n13) -> (Output_Get:double)
   Set Is emote?(execute=@n12.else, Is emote?=@n13)
   TryGetPawnOwner()
  ]}
]
variables [
 Celeb up {type=boolean}
 Celeb up alpha {type=number}
 headRotLimit {type=Rotator}
 Is emote? {type=number}
 X percentageEmote {type=number, default=0.4}
 Y percentageEmote {type=number, default=0.8}
 Z percentageEmote {type=number, default=0.8}
]
